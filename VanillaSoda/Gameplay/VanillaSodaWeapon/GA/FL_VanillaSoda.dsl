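; Function library for the VanillaSoda weapon's abilities, with helpers to orient and spawn arc projectiles.
; ThrowArcProjectile rolls the spawn rotation about its forward axis via AdjustSpawnRotation, passing the untilted rotation alongside it.
; It spawns client-predicted if the class casts to a FortProjectileAthena class, otherwise normally with the same params; its Scale input is unused.
; Separately, AdjustArcProjectileSpawnDirection offsets forward along local right and world up, returning the resulting rotation.
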(fn AdjustArcProjectileSpawnDirection
  (params (BaseRotation :type Rotator) (Direction%20Offset%20Right :type Float%20%28double-precision%29) (Direction%20Offset%20Up :type Float%20%28double-precision%29))
  (return (Math|Conversions|RotationFromXVector (Math|Vector|Normalize (+ (* Direction%20Offset%20Right (Math|Vector|GetRightVector BaseRotation)) (Utilities|Operators|Add (Math|Vector|GetForwardVector BaseRotation) (Math|Vector|MakeVector 0.0 0.0 Direction%20Offset%20Up)))))))

(fn AdjustSpawnRotation
  (params (InputPin :type Rotator) (AngleDeg :type Float%20%28single-precision%29))
  (return (Math|Rotator|MakeRotfromYX (Math|Vector|RotateVectorAroundAxis (Math|Vector|GetRightVector InputPin) AngleDeg (Math|Vector|GetForwardVector InputPin)) (Math|Vector|GetForwardVector InputPin))))

(fn ThrowArcProjectile
  (params (Class :type Fort%20Projectile%20Base%20Class%20Reference) (Scale :type Vector) (SpawnLocation :type Vector) (Player :type Actor%20Object%20Reference) (Rotation :type Rotator) (TiltAngleDeg :type Float%20%28single-precision%29) (InitialSpeed :type Float%20%28single-precision%29 :default 8e+03))
  (bind _returnvalue (Private|AdjustSpawnRotation Rotation TiltAngleDeg))
  (bind _spawnprojectileparams (Utilities|Struct|MakeSpawnProjectileParams SpawnLocation _returnvalue Rotation "(TargetSelection=(List=,bStopAtFirstSuccess=True,bKeepCheckingListOnIndestructibleHit=False,AbilityTargetBlockedGameplayTags=(GameplayTags=),bUseWeaponRanges=True,bUseMaxYawAngleToTarget=False,MaxYawAngleToTarget=0.000000),TargetGameplayEffectSpecs=,OwnerGameplayEffectSpecs=,ActivationCues=(GameplayTags=),ImpactCues=(GameplayTags=),ImpactNoiseRange=0.000000,FlyByNoiseRange=0.000000,bOverrideChargeMagnitude=False,ChargeMagnitudeOverrideValue=0.000000)" "(TargetSelection=(List=,bStopAtFirstSuccess=True,bKeepCheckingListOnIndestructibleHit=False,AbilityTargetBlockedGameplayTags=(GameplayTags=),bUseWeaponRanges=True,bUseMaxYawAngleToTarget=False,MaxYawAngleToTarget=0.000000),TargetGameplayEffectSpecs=,OwnerGameplayEffectSpecs=,ActivationCues=(GameplayTags=),ImpactCues=(GameplayTags=),ImpactNoiseRange=0.000000,FlyByNoiseRange=0.000000,bOverrideChargeMagnitude=False,ChargeMagnitudeOverrideValue=0.000000)" false 0 InitialSpeed 0.0 0.0))
  (bind _player (Variables|GetPlayer))
  (bind _asfort_projectile_athena (Utilities|Casting|CastToFortProjectileAthenaClass Class)
    (:then
      (bind _returnvalue_2 (Projectile|SpawnClientPredictedProjectile _asfort_projectile_athena _player _spawnprojectileparams))
      (return _returnvalue_2))
    (:CastFailed
      (bind _returnvalue_3 (Projectile|SpawnProjectilewithParams Class _player _spawnprojectileparams))
      (return _returnvalue_3))))
