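; Mantis melee gameplay ability for the Vanilla Soda weapon, driving its combo swings, heavy attack and air slam.
; Technique events and animation notifies drive camera, force feedback, hover and guided-motion fall; landing, tag waits or a timeout end the fall.
; Landing starts the slam technique, damaging pawns then world through throttled async targeting.
; Helpers handle projectile aim assist, warp offsets, montage choice, and deferred durability cost for heavy and slam.
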(event Ability|EventActivateAbility
  ;[0.07613 0.2073 0.04981] Cache Components and Hotfixes
  (|CallParentFunction)
  ;; Used to track if certain abilites were used
  ;; and then deduct durability at the appropriate time.
  (Variables|Transient|SetbHeavyAttackProjectileSpawned false)
  (Variables|Transient|SetbAirSlamUsed false)
  (General|CacheComponents)
  (bind _returnvalue (Ability|ApplyGameplayEffectToOwner (Variables|GameplayEffects|GetDisableSwimSprintingGameplayEffectClass)))
  (Variables|Transient|SetDisableSwimSprintingGameplayEffectHandle _returnvalue)
  (Camera|SetEnableMantisCameraModifier true)
  ;; Canceling sprint after ability has started so that GetDynamicMontageNameForTechnique can run accurately.
  (Class|FortMovementCompCharacter|CancelActiveExtensionwithTag (Variables|Default|GetFort%20Movement%20Comp%20Character) (Variables|GameplayTags|GetTacSprintGameplayTags)))

(event Ability|EventOnEndAbility
  (params (bWasCancelled :type Boolean))
  ;[0.15 0.0705 0.01555 0.5] EndAbility: Clean up.
  (Utilities|FlowControl|Sequence
    (:then_0
      (Ability|RemoveGameplayEffectFromOwnerWithHandle (Variables|Transient|GetDisableSwimSprintingGameplayEffectHandle))
      (Camera|SetEnableMantisCameraModifier false)
      (CallFunction|RemoveMidAirHold)
      (General|ToggleCanChangeWeapons true)
      (Ability|RemoveGameplayEffectFromOwnerWithHandle (Variables|Transient|GetPreventMovementGameplayEffectHandle))
      (Ability|RemoveGameplayEffectFromOwnerWithHandle (Variables|Transient|GetPreventEquipmentChangeGameplayEffectHandle))
      (Class|FortMovementModeExtBlueprintLibrary|EndMovementModeExtensiononActor (Variables|Transient|GetAirHoldMovementModeRuntimeData) (Variables|Transient|GetAvatarPlayerPawn))
      (bind _airslamdivecameramodifierhandle (Variables|Transient|GetAirSlamDiveCameraModifierHandle)
        (:then
          (Class|FortCameraModifierTargetRotationFrustum|RemoveSelffromCameraManager _airslamdivecameramodifierhandle))
        (:else
          _))
      (General|SetEnableHighAccelerationTolerance false)
      (Class|FortMovementModeExtBlueprintLibrary|EndMovementModeExtensiononActor (Variables|Transient|GetAirAttackFallMovementModeExtRuntimeData) (Variables|Transient|GetAvatarPlayerPawn))
      (GameplayTasks|EndTask (Variables|Transient|GetAsyncTasks_WaitGameplayTagAdd))
      (bind _avatarplayerpawn (Variables|Transient|GetAvatarPlayerPawn)
        (:then
          (EventDispatchers|UnbindEventfromOnPawnLanded _avatarplayerpawn (EventDispatchers|CreateEvent :!event Custom|OnLanded)))
        (:else
          _))
      (Utilities|Time|ClearandInvalidateTimerbyHandle (Variables|Transient|GetPostAirSlamLockoutTimerHandle))
      (bind _asynctask_waitlandtimeout (Variables|Transient|GetAsyncTask_WaitLandTimeout)
        (:then
          (GameplayTasks|EndTask _asynctask_waitlandtimeout))
        (:else
          _))
      (bind _returnvalue (General|GetCameraAnimationModifier))
      (Utilities|IsValid _returnvalue
        (:Is%20Valid
          (CameraAnimation|StopCameraAnimation _returnvalue (Variables|Transient|GetSlamLandingCameraAnimationHandle) false :!class /Script/EngineCameras.CameraAnimationCameraModifier)
          (CameraAnimation|StopCameraAnimation _returnvalue (Variables|Transient|GetHeavyCameraAnimationHandle) false :!class /Script/EngineCameras.CameraAnimationCameraModifier)
          (CameraAnimation|StopCameraAnimation _returnvalue (Variables|Transient|GetSlamHoverCameraAnimationHandle) false :!class /Script/EngineCameras.CameraAnimationCameraModifier))))
    (:then_1
      (|CallParentFunction bWasCancelled))))

(event Mantis|EventOnMantisAblilityTechniqueStarted
  (params (OwningTags :type Gameplay%20Tag%20Container%20Structure) (SequenceName :type Name :default "None") (StartsSequence :type Boolean) (EndsSequence :type Boolean))
  (bind _hotfix_maxtargetstoprocesspertick_slam (Variables|Hotfix|GetHotfix_MaxTargetsToProcessPerTick_Slam))
  ;[0 0.1249 0.3615] On Technique Started
  (|CallParentFunction OwningTags SequenceName StartsSequence EndsSequence)
  (switch name SequenceName
    (:Combo1
      (bind _mantiscameramodifier (Variables|Transient|GetMantisCameraModifier)
        (:then
          (Class|CameraModifierVanillaSodaMantis|NotifyComboStart _mantiscameramodifier 0))
        (:else)))
    (:Combo2
      (bind _mantiscameramodifier_2 (Variables|Transient|GetMantisCameraModifier)
        (:then
          (Class|CameraModifierVanillaSodaMantis|NotifyComboStart _mantiscameramodifier_2 1))
        (:else)))
    (:AirSlam
      ;;[0.1472 0 0.15 0.5] Activating the air slam attack.
      (General|ToggleCanChangeWeapons false)
      (Ability|HasAuthority
        (:True
          _)
        (:False
          (bind _returnvalue (Game|ApplyGuidedMotionMME (Variables|Transient|GetAvatarPlayerPawn) "/VanillaSoda/Gameplay/VanillaSodaWeapon/Movement/FMM_GuidedMotion_VanillaSodaMovement_AirSlamHoldBeforeFalling.FMM_GuidedMotion_VanillaSodaMovement_AirSlamHoldBeforeFalling_C"))
          (Variables|Transient|SetAirHoldMovementModeRuntimeData _returnvalue)))
      (bind _returnvalue_2 (Ability|ApplyGameplayEffectToOwner (Variables|GameplayEffects|GetDivingGameplayEffectClass)))
      (Variables|Transient|SetDivingGameplayEffectHandle _returnvalue_2)
      (ForceFeedback|PlayForceFeedback (Variables|ForceFeedback|GetForceFeedbackAirSlamStart))
      (Camera|PlaySlamHoverCameraAnimation))
    (:Landing
      ;;[0 0.15 0.08734 0.5] Perform landing damage from the air slam.
      (bind (_asynctaskproxy _targetdata) (Ability|Tasks|PerformTargetingAbilityTask "LandingTargeting_Pawns" (Variables|TargetingPreset|GetTargetingPreset_SlamPawn) true false true)
        (:then
          (Variables|Transient|SetbAirSlamUsed true)
          (Camera|PlaySlamLandingCameraAnimation)
          (ForceFeedback|PlayForceFeedback (Variables|ForceFeedback|GetForceFeedbackAirSlamImpact)))
        (:Targeted
          _)
        (:Cancelled
          _))
      (Ability|Tasks|ApplyGameplayEffectToTargetDataAsync _targetdata (Variables|GameplayEffects|GetSlamDamagePawnGameplayEffectClass) 1 1 _hotfix_maxtargetstoprocesspertick_slam
        (:OnCompleted
          (bind (_asynctaskproxy_2 _targetdata_2) (Ability|Tasks|PerformTargetingAbilityTask "LandingTargeting_World" (Variables|TargetingPreset|GetTargetingPreset_SlamWorld) true false true)
            (:Targeted
              (Ability|Tasks|ApplyGameplayEffectToTargetDataAsync _targetdata_2 (Variables|GameplayEffects|GetSlamDamageWorldGameplayEffectClass) 1 1 _hotfix_maxtargetstoprocesspertick_slam))))))))

(event Custom|ApplyMidAirHoldSlash
  ;[0 0.15 0.1348 0.5] ApplyMidAirHoldSlash: Invoke the mid air hold guided motion MME.
  (bind _avatarplayerpawn (Variables|Transient|GetAvatarPlayerPawn)
    (:then
      (if (not (Variables|/Script/FortniteGame.FortPawn|GetbIsSliding _avatarplayerpawn))
        (Ability|HasAuthority
          (:True
            _)
          (:False
            _))
        (bind _returnvalue (Game|ApplyGuidedMotionMME (Variables|Transient|GetAvatarPlayerPawn) (Variables|Heavy|GetHeavyAbilityGuidedMotion)))))
    (:else)))

(event Ability|EventOnGenericNotifyFromAnimation
  (params (NotifyTag :type Gameplay%20Tag%20Structure))
  ;[0.1268 0.0313 0.15 0.5] OnGenericNotifyFromAnimation: React to notifies fired from the animations.
  (|CallParentFunction NotifyTag)
  (switch Utilities|FlowControl|Switch|SwitchonGameplayTag NotifyTag
    (:Animation.Notify.LemonExam.ProjectileThrow
      (Utilities|FlowControl|Sequence
        (:then_0
          ;;[0.003465 0.15 0.1322 0.5] Cache the melee hit trace hit result. This will be passed to the projectile so it ignores whatever this hits.
          (Ability|HasAuthority
            (:True
              (bind (_asynctaskproxy _targetdata) (Ability|Tasks|DoTargetSelection "HeavyMeleeHit" "(TagName=\"Gameplay.Melee.Focused\")" true)
                (:Targeted
                  (Ability|Tasks|ApplyGameplayEffectToTargetDataAsync _targetdata (Variables|Heavy|GetHeavyAttackDamageClass) 1 1 (Variables|Hotfix|GetHotfix_MaxTargetsToProcessPerTick_HeavyClose)))
                (:Cancelled
                  _))
              (Heavy|CacheHeavyHitMeleeTargetHitResults _targetdata))))
        (:then_1
          (ForceFeedback|PlayForceFeedback (Variables|ForceFeedback|GetForceFeedbackHeavyFire)))))
    (:Animation.Notify.LemonExam.RemoveHold
      (CallFunction|RemoveMidAirHold))
    (:Animation.Notify.LemonExam.Slam.Fall.Start
      (CallFunction|BeginAirAttackFall))))

(event Custom|RemoveMidAirHold
  ;[0.1414 0.007014 0.15 0.5] RemoveMidAirHold: Allow changing of weapons, remove gameplay effects that were blocking move and switching, uncrouch the player pawn.
  ; This should clean up for both the mid air hold of the air slam as well as the heavy attack.
  (General|ToggleCanChangeWeapons true)
  (General|RemovePreventMoveandSwitch)
  (Character|UnCrouch (Variables|Transient|GetAvatarPlayerPawn)))

(event Mantis|EventOnMantisAblilityTechniqueEnded
  (params (SequenceName :type Name :default "None"))
  ;[0.0206 0.1451 0.15 0.5] OnMantisAbilityTechniqueEnded: The technique has ended. Run desired logic.
  (switch name SequenceName
    (:Combo1
      (bind _mantiscameramodifier (Variables|Transient|GetMantisCameraModifier)
        (:then
          (Class|CameraModifierVanillaSodaMantis|NotifyComboEnd _mantiscameramodifier 0))
        (:else)))
    (:Combo2
      (bind _mantiscameramodifier_2 (Variables|Transient|GetMantisCameraModifier)
        (:then
          (Class|CameraModifierVanillaSodaMantis|NotifyComboEnd _mantiscameramodifier_2 1))
        (:else)))
    (:Combo3
      (bind _mantiscameramodifier_3 (Variables|Transient|GetMantisCameraModifier)
        (:then
          (Class|CameraModifierVanillaSodaMantis|NotifyComboEnd _mantiscameramodifier_3 2))
        (:else)))))

(event Custom|BeginAirAttackFall
  ;[0.07947 0 0.08333] Perform the ground slam attack.
  ;;[0 0.07122 0.15 0.5] BeginAirAttackFall: End the air movement hold, remove camera modifier, start falling attack if the character is in the falling movement mode.
  (Class|FortMovementModeExtBlueprintLibrary|EndMovementModeExtensiononActor (Variables|Transient|GetAirHoldMovementModeRuntimeData) (Variables|Transient|GetAvatarPlayerPawn))
  (Class|FortCameraModifierTargetRotationFrustum|RemoveSelffromCameraManager (Variables|Transient|GetAirSlamDiveCameraModifierHandle)) ; Give player control back before they land
  (if (GameplayTags|HasMatchingGameplayTag (Variables|Transient|GetAvatarPlayerPawn) (Variables|GameplayTags|GetInWaterTag))
    (Ability|EndAbility)
    (elif (GameplayTags|HasMatchingGameplayTag (Variables|Transient|GetAvatarPlayerPawn) "(TagName=\"MovementMode.Falling\")")
      (CallFunction|DoFallingAttack)
      (else
        (CallFunction|OnLanded)))))

(event Custom|DoFallingAttack
  (bind _asynctasks_waitgameplaytagadd (Variables|Transient|GetAsyncTasks_WaitGameplayTagAdd))
  ;[0.07947 0 0.08333] Perform the ground slam attack.
  ;;[0.04707 0.15 0.1476 0.5] DoFallingAttack: Start the falling MME and bind to know when the character has landed.
  (Ability|ApplyGameplayEffectToOwner (Variables|Default|GetAthenaFallDamageImmunity))
  (General|SetEnableHighAccelerationTolerance true)
  (CallFunction|DoMMESlam)
  (EventDispatchers|BindEventtoOnPawnLanded (Variables|Transient|GetAvatarPlayerPawn) (EventDispatchers|CreateEvent :!event Custom|OnLanded))
  (bind _asynctaskproxy (Ability|Tasks|WaitGameplayTagAdd (Variables|GameplayTags|GetSlopeSlidingTag) 0 true)
    (:then
      (Utilities|Array|Add _asynctasks_waitgameplaytagadd _asynctaskproxy)
      (bind _asynctaskproxy_2 (Ability|Tasks|WaitGameplayTagAdd (Variables|GameplayTags|GetInWaterTag) 0 true)
        (:then
          (Utilities|Array|Add _asynctasks_waitgameplaytagadd _asynctaskproxy_2)
          (bind _asynctaskproxy_3 (Ability|Tasks|WaitGameplayTagAdd (Variables|GameplayTags|GetRecentlyUsedCrashpadTag) 0 true)
            (:then
              (Utilities|Array|Add _asynctasks_waitgameplaytagadd _asynctaskproxy_3)
              ;;[0 0.15 0.1354 0.5] A timeout duration in case land is never invoked.
              (bind _asynctaskproxy_4 (Ability|Tasks|WaitDelay (ScalableFloat|GetCachedScalableFloatValue (Variables|Hotfix|GetHotfix_FallAttackWaitLandTimeoutDuration)))
                (:then
                  (Variables|Transient|SetAsyncTask_WaitLandTimeout _asynctaskproxy_4))
                (:OnFinish
                  (Ability|EndAbility))))
            (:Added
              (Class|FortMovementModeExtBlueprintLibrary|EndMovementModeExtensiononActor (Variables|Transient|GetAirAttackFallMovementModeExtRuntimeData) (Variables|Transient|GetAvatarPlayerPawn))
              (Ability|EndAbility))))
        (:Added
          (Default|WaterDepthCheck
            (:End
              (Ability|EndAbility))
            (:Don't%20End
              (CallFunction|OnLanded)
              (Ability|ExecuteGameplayCueWithParamsOnOwner (Utilities|Struct|BreakGameplayCueTag (Variables|GameplayCueTag|GetAirSlamEnterWaterGameplayCueTag)) (Ability|GameplayCue|MakeGameplayCueParameters 0.0)))))))
    (:Added
      (CallFunction|OnLanded))))

(event Custom|DoLaunchSlam
  (bind _avatarplayerpawn (Variables|Transient|GetAvatarPlayerPawn))
  ;[0.07947 0 0.08333] Perform the ground slam attack.
  (bind _targetlocation (GroundSlam|FindSlamTarget))
  (Rendering|Debug|DrawDebugSphere _targetlocation 75.0 12 "(R=1.000000,G=1.000000,B=1.000000,A=1.000000)" 1e+01 3.0)
  (bind _outlaunchvelocity (Game|SuggestProjectileVelocityCustomArc (Transformation|GetActorLocation _avatarplayerpawn) _targetlocation 0.0 0.9))
  (Character|LaunchCharacter _avatarplayerpawn _outlaunchvelocity true true))

(event Custom|OnLanded
  (params (Hit :type Hit%20Result%20Structure%20%28by%20ref%29))
  ;[0.07947 0 0.08333] Perform the ground slam attack.
  ;;[0 0.1284 0.15 0.5] OnLanded: Called at the end of the fall of the air slam attack.
  (Utilities|FlowControl|Sequence
    (:then_0
      ;;; Unbind and remove high acceleration tolerance.
      (EventDispatchers|UnbindEventfromOnPawnLanded (Variables|Transient|GetAvatarPlayerPawn) (EventDispatchers|CreateEvent :!event Custom|OnLanded))
      (General|SetEnableHighAccelerationTolerance false)
      (GameplayTasks|EndTask (Variables|Transient|GetAsyncTasks_WaitGameplayTagAdd))
      (bind _asynctask_waitlandtimeout (Variables|Transient|GetAsyncTask_WaitLandTimeout)
        (:then
          (GameplayTasks|EndTask _asynctask_waitlandtimeout))
        (:else
          _))
      (bind _slamtask (Variables|Default|GetSlamTask)
        (:then
          (GameplayTasks|EndTask _slamtask))
        (:else)))
    (:then_1
      ;;;[0.04499 0.07226 0.15 0.5] This jumps us to the landed technique where the damage and knockback actually takes place.
      (bind _mantis_component (Variables|Default|GetMantis%20Component)
        (:then
          (Mantis|SetBranchRule _mantis_component "Any" "Default" false)
          (Mantis|TryManuallyStartTechnique _mantis_component 2)
          (Ability|ApplyGameplayEffectToOwner (Variables|GameplayEffects|GetSlamFeedbackGE)))
        (:else)))
    (:then_2
      ;;;[0.03038 0.15 0.1264 0.5] Once landed, end the falling attack MME.
      (Class|FortMovementModeExtBlueprintLibrary|EndMovementModeExtensiononActor (Variables|Transient|GetAirAttackFallMovementModeExtRuntimeData) (Variables|Transient|GetAvatarPlayerPawn)))))

(event Custom|DoMMESlam
  ;[0.07947 0 0.08333] Perform the ground slam attack.
  (bind _returnvalue (Game|ApplyGuidedMotionMME (Variables|Transient|GetAvatarPlayerPawn) (Variables|Default|GetFallingGuidedMotion)))
  (Variables|Transient|SetAirAttackFallMovementModeExtRuntimeData _returnvalue))

(event Mantis|EventOnTechniqueNotifyWindowStarted
  (params (NotifyWindow :type EFortMantisNotifyWindow%20Enum :default "Invalid") (Notify :type Mantis%20Window%20Object%20Reference))
  ;[0 0.15 0.0634 0.5] OnTechniqueNotifyWindowStarted: Apply execution activate gameplay effect to the owner.
  (|CallParentFunction NotifyWindow Notify)
  (if (Utilities|Enum|Equal(Enum) NotifyWindow "Execution")
    (Ability|ApplyGameplayEffectToOwner (Variables|GameplayEffects|GetMeleeExecutionWindowActivateGameplayEffectClass))))

(fn AdjustAim
  (params (ProjectileSpawnLocation :type Vector))
  (bind _originalspawnrotation (LocalVariables|GetOriginalSpawnRotation))
  (bind _returnvalue (Math|Vector|GetForwardVector _originalspawnrotation))
  (LocalVariables|SetOriginalSpawnRotation (Default|AdjustArcProjectileSpawnDirection (Heavy|GetCamRotation) (ScalableFloat|GetCachedScalableFloatValue (Variables|Hotfix|GetHotfix_ProjectileDirectionOffsetRight)) (ScalableFloat|GetCachedScalableFloatValue (Variables|Hotfix|GetHotfix_ProjectileDirectionOffsetUp))))
  (Utilities|FlowControl|Sequence
    (:then_0
      ;[0.149 0.3059 0.8941] Adjust aim to targeted player (if enabled)
      ; Because the projectile fires from around midheight of the player, it doesn't always hit what is under the reticle
      ; when aiming at something very close (eg dbno player at feet). To counteract this, search for and add a bit of
      ; auto aim to nearby players.
      (if (> (ScalableFloat|GetCachedScalableFloatValue (Variables|Hotfix|GetHotfix_AdjustAimToTargetedPlayer)) 0)
        (bind _current_weapon (Variables|CachedValues|GetCurrent%20Weapon)
          (:then
            (FortWeapon|UpdateDamageStartLocation _current_weapon) ; Update it before getting as it may not have been set yet on the server
            (bind _output_get (LocalVariables|SetTraceStart (.ReturnValue (FortWeapon|GetDamageStartLocation _current_weapon -1))))
            (bind (_outhit _returnvalue_2) (Collision|SphereTraceForObjects _output_get (+ _output_get (* (FortWeapon|GetDamageStartLocation _current_weapon -1) (ScalableFloat|GetCachedScalableFloatValue (Variables|Hotfix|GetHotfix_AdjustAimToTargetedPlayerTraceDist)))) (ScalableFloat|GetCachedScalableFloatValue (Variables|Hotfix|GetHotfix_AdjustAimToTargetedPlayerTraceRadius)) (Utilities|Array|MakeArray "ObjectTypeQuery3") false (Utilities|Array|MakeArray (Ability|GetActivatingPawn)) (select (LocalVariables|GetbDebugDrawAimTrace) "ForDuration" "None") true "(R=0.000000,G=0.477117,B=1.000000,A=1.000000)" "(R=0.237847,G=0.005192,B=0.036048,A=1.000000)")) ; Since the trace radius should be fairly small\nand this is mostly to handle issues with nearby pawns,\njust single trace and use the nearest hit.
            (if _returnvalue_2
              (bind _output_get_2 (LocalVariables|SetRotationToHitPawn (Math|Rotator|MakeRotfromXY (- (.ImpactPoint (Collision|BreakHitResult _outhit)) (Variables|GetProjectileSpawnLocation)) (Math|Vector|GetRightVector (LocalVariables|GetOriginalSpawnRotation)))))
              (LocalVariables|SetValidHit (> (Math|Vector|DotProduct (Math|Vector|GetForwardVector _output_get_2) _returnvalue) (ScalableFloat|GetCachedScalableFloatValue (Variables|Hotfix|GetHotfix_AdjustAimToTargetedPlayerMinDot))))
              (if (LocalVariables|GetbDebugDrawAimAdjustment)
                ;;[0.2066 0.2066 0.2066] Debug
                (Development|PrintString (Utilities|String|Append "Dot " (Utilities|String|ToString(Float) (Math|Vector|DotProduct (Math|Vector|GetForwardVector _output_get_2) _returnvalue))))
                (Rendering|Debug|DrawDebugSphere (.ImpactPoint (Collision|BreakHitResult _outhit)) 1e+01 12 "(R=1.000000,G=0.015551,B=0.000000,A=1.000000)" 2.0) ; impact point on hit pawn
                (Rendering|Debug|DrawDebugLine (Variables|GetProjectileSpawnLocation) (+ (Variables|GetProjectileSpawnLocation) (* (Math|Vector|GetForwardVector (LocalVariables|GetOriginalSpawnRotation)) (ScalableFloat|GetCachedScalableFloatValue (Variables|Hotfix|GetHotfix_AdjustAimToTargetedPlayerTraceDist)))) "(R=0.000000,G=0.098039,B=1.000000,A=1.000000)" 2.0 1.0) ; Original rotation forward vector
                (Rendering|Debug|DrawDebugLine (Variables|GetProjectileSpawnLocation) (+ (Variables|GetProjectileSpawnLocation) (* (Math|Vector|GetForwardVector (LocalVariables|GetRotationToHitPawn)) (ScalableFloat|GetCachedScalableFloatValue (Variables|Hotfix|GetHotfix_AdjustAimToTargetedPlayerTraceDist)))) (Math|Color|SelectColor "(R=0.066546,G=1.000000,B=0.000000,A=1.000000)" "(R=0.616320,G=0.301948,B=0.027567,A=1.000000)" (LocalVariables|GetValidHit)) 2.0 1.0) ; Potential new rotation foward  vector
              )
              (if (LocalVariables|GetValidHit)
                ;;[1 0.6745 0.2784] Return new rotation
                (return true (LocalVariables|GetRotationToHitPawn)))))
          (:else))))
    (:then_1
      ;[1 0.6745 0.2784] Return original rotation
      (return false (LocalVariables|GetOriginalSpawnRotation)))))

(fn CacheComponents
  ;[3.6e-05 0.1405 0.15 0.5] Cache relevant components and hotfixes for this ability.
  (bind _asfort_player_pawn (Utilities|Casting|CastToFortPlayerPawn (Ability|GetActivatingPawn)))
  (bind _output_get (Variables|Transient|SetAvatarPlayerPawn _asfort_player_pawn))
  (bind _asfort_movement_comp_character (Utilities|Casting|CastToFortMovementComp_Character (Pawn|GetMovementComponent _output_get)))
  (Variables|Default|SetFort%20Movement%20Comp%20Character _asfort_movement_comp_character)
  (bind _asb_vanilla_soda_weapon (Utilities|Casting|CastToB_VanillaSoda_Weapon (Weapon|GetCurrentWeapon (Variables|Transient|GetAvatarPlayerPawn))))
  (Variables|CachedValues|SetCurrent%20Weapon _asb_vanilla_soda_weapon)
  (Variables|Default|SetMantis%20Component (Actor|GetComponentbyClass (Variables|Transient|GetAvatarPlayerPawn) "/Script/MantisRuntime.FortMantisPawnComponent")))

(fn CacheHeavyHitMeleeTargetHitResults
  (params (TargetDataHandle :type Gameplay%20Ability%20Target%20Data%20Handle%20Structure))
  (Utilities|Array|Clear (Variables|Transient|GetHeavyMeleeHitResults))
  (for _index (range (Ability|TargetData|GetDataCountfromTargetData TargetDataHandle))
    (if (Ability|TargetData|TargetDataHasHitResult TargetDataHandle _index)
      (Utilities|Array|Add (Variables|Transient|GetHeavyMeleeHitResults) (Ability|TargetData|GetHitResultfromTargetData TargetDataHandle _index)))))

(fn CameraModHeightCheck
  (bind _hotfix_minheightforcameramod (Variables|Hotfix|GetHotfix_MinHeightForCameraMod))
  (bind _returnvalue (ScalableFloat|GetValueatLevel _hotfix_minheightforcameramod))
  (if (< _returnvalue 0)
    (return false)
    (else
      (bind (_outhit _returnvalue_2) (Collision|LineTraceByChannel (Transformation|GetActorLocation (Variables|Transient|GetAvatarPlayerPawn)) (Utilities|Operators|Subtract (Transformation|GetActorLocation (Variables|Transient|GetAvatarPlayerPawn)) (Math|Vector|MakeVector 0.0 0.0 _returnvalue)) "TraceTypeQuery10" false (Utilities|Array|MakeArray (Variables|Transient|GetAvatarPlayerPawn)) "ForDuration"))
      (return (not _returnvalue_2)))))

(fn CommitDurabilityCost
  ;[0 0.05058 0.15 0.5] CommitDurabilityCost: Commit the durability cost of the heavy attack if the projectile was spawned.
  (|CallParentFunction)
  (Ability|HasAuthority
    (:True
      (Utilities|FlowControl|Sequence
        (:then_0
          (if (Variables|Transient|GetbHeavyAttackProjectileSpawned)
            (Variables|Transient|SetbHeavyAttackProjectileSpawned false) ; Reset
            (bind _current_weapon (Variables|CachedValues|GetCurrent%20Weapon)
              (:then
                (FortWeapon|UseWeaponDurability(Scaled) _current_weapon (ScalableFloat|GetCachedScalableFloatValue (Variables|Hotfix|GetHotfix_DurabilityCostHeavy))))
              (:else))))
        (:then_1
          (if (Variables|Transient|GetbAirSlamUsed)
            (Variables|Transient|SetbAirSlamUsed false) ; Reset
            (bind _current_weapon_2 (Variables|CachedValues|GetCurrent%20Weapon)
              (:then
                (FortWeapon|UseWeaponDurability(Scaled) _current_weapon_2 (ScalableFloat|GetCachedScalableFloatValue (Variables|Hotfix|GetHotfix_DurabilityCostAirSlam))))
              (:else))))))))

(fn ConfigureRootMotionWarpingAlternativeBehavior
  (params (TechniqueName :type Name :default "None"))
  (bind (_allowtranslationwarp _allowrotationwarp _snapshottargetlocation) (CallFunction|GetRootMotionOptions))
  (bind _aimassisttarget (Variables|Default|GetAimAssistTarget))
  ;[0.07483 0.0513 0.15 0.5] Overridden to set an offset to the target actor while warping.
  (Utilities|FlowControl|Sequence
    (:then_0
      (CallFunction|GetAimAssistTarget)
      (if (Variables|Default|GetHasAimAssistTarget)
        (bind _returnvalue (CallFunction|IsValidTarget _aimassisttarget))
        (if _returnvalue
          (return _aimassisttarget _snapshottargetlocation "OffsetFromTargetActor" (ScalableFloat|GetValueatLevel (Variables|Hotfix|GetHotfx_MinimumTargetOffset)) 0.0 0.0 "FaceTargetActor" "0, 0, 0"))))
    (:then_1
      (Development|PrintString "No Target")
      (return 0 false "None" "0, 0, 0" "None" "0, 0, 0"))))

(fn FindSlamTarget
  (bind _outhit (Collision|SphereTraceByChannel (Camera|GetCameraLocation (Variables|/Script/Engine.PlayerController|GetPlayerCameraManager (Fortnite|GetFortPlayerControllerfromActor (Variables|Transient|GetAvatarPlayerPawn)))) (+ (Camera|GetCameraLocation (Variables|/Script/Engine.PlayerController|GetPlayerCameraManager (Fortnite|GetFortPlayerControllerfromActor (Variables|Transient|GetAvatarPlayerPawn)))) (* (GroundSlam|GetSlamDirection) 1e+03)) (GroundSlam|GetSlamTargetTraceRadius) "TraceTypeQuery1" true (GroundSlam|GetSlamTargetTraceActorstoIgnore) "ForDuration" true "(R=0.000000,G=0.563464,B=1.000000,A=1.000000)" "(R=1.000000,G=0.000000,B=0.926133,A=1.000000)" 1e+01))
  (return (Math|Vector|SelectVector (.ImpactPoint (Collision|BreakHitResult _outhit)) (.TraceEnd (Collision|BreakHitResult _outhit)) (Collision|BreakHitResult _outhit)) (.HitActor (Collision|BreakHitResult _outhit))))

(fn GetCameraAnimationModifier
  (Utilities|FlowControl|Sequence
    (:then_0
      (bind _avatarplayerpawn (Variables|Transient|GetAvatarPlayerPawn)
        (:then
          (Utilities|IsValid (Fortnite|GetFortPlayerControllerfromActor _avatarplayerpawn)
            (:Is%20Valid
              (bind _playercameramanager (Variables|/Script/Engine.PlayerController|GetPlayerCameraManager (Fortnite|GetFortPlayerControllerfromActor _avatarplayerpawn))
                (:then
                  (return (CameraAnimation|ConvCameraAnimationCameraModifier _playercameramanager)))
                (:else)))))
        (:else)))
    (:then_1
      (return 0))))

(fn GetCamRotation
  (bind _returnvalue (Ability|GetActivatingPawn))
  (bind _returnvalue_2 (Fortnite|GetFortPlayerControllerfromActor _returnvalue))
  (Utilities|IsValid _returnvalue_2
    (:Is%20Valid
      (return (Camera|GetCameraRotation (Variables|/Script/Engine.PlayerController|GetPlayerCameraManager _returnvalue_2))))
    (:Is%20Not%20Valid
      (bind _asfort_athena_aibot_controller (Utilities|Casting|CastToFortAthenaAIBotController (Pawn|GetController _returnvalue)))
      (return (Pawn|GetControlRotation _asfort_athena_aibot_controller :!class /Script/Engine.Controller)))))

(fn GetDynamicMontageNameForTechnique
  (params (TechniqueName :type Name :default "None"))
  (switch name TechniqueName
    (:Default
      (bind _returnvalue (|CallParentFunction TechniqueName))
      (return _returnvalue))
    (:Combo1
      (bind _asfort_player_pawn (Utilities|Casting|CastToFortPlayerPawn (Variables|Transient|GetAvatarPlayerPawn)))
      (if (Variables|/Script/FortniteGame.FortPlayerPawn|GetbIsTacticalSprinting _asfort_player_pawn)
        (return "TacSprinting")
        (else
          (return "Normal"))))
    (:Heavy
      (bind _asfort_player_pawn_2 (Utilities|Casting|CastToFortPlayerPawn (Variables|Transient|GetAvatarPlayerPawn)))
      (if (Variables|/Script/FortniteGame.FortPlayerPawn|GetbIsTacticalSprinting _asfort_player_pawn_2)
        (return "TacSprinting")
        (elif (PawnStats|IsSprinting _asfort_player_pawn_2)
          (return "Running")
          (else
            (return "Normal")))))))

(fn GetMinDotWithDown
  (params (DistanceAboveGround :type Float%20%28double-precision%29))
  (return (Variables|Default|GetMinAllowedDotProduct)))

(fn GetSlamDirection
  (bind _avatarplayerpawn (Variables|Transient|GetAvatarPlayerPawn))
  (bind _returnvalue (Fortnite|GetFortPlayerControllerfromActor _avatarplayerpawn))
  (bind (_location _rotation) (Pawn|GetPlayerViewPoint _returnvalue))
  (bind _returnvalue_2 (Math|Vector|GetForwardVector _rotation))
  (bind (_returnvalue_3 _outmindot) (GroundSlam|IsDirectionWithinAllowedAngle _returnvalue_2))
  (LocalVariables|SetOut_Direction _returnvalue_2)
  (if (not _returnvalue_3)
    (LocalVariables|SetOut_Direction (Math|Vector|RotateVectorAroundAxis (Transformation|GetActorForwardVector _avatarplayerpawn) (Math|Float|MapRangeClamped _outmindot -1.0 1.0 -9e+01 9e+01) (Transformation|GetActorRightVector _avatarplayerpawn))))
  (return (LocalVariables|GetOut_Direction)))

(fn GetSlamTargetTraceActorsToIgnore
  (return (Utilities|Array|MakeArray (Variables|Transient|GetAvatarPlayerPawn))))

(fn GetSlamTargetTraceEnd
  (return (+ (GroundSlam|GetSlamTargetTraceStart) (* (GroundSlam|GetSlamDirection) (ScalableFloat|GetValueatLevel (Variables|Default|GetDashMaxDistance))))))

(fn GetSlamTargetTraceRadius
  (return (ScalableFloat|GetValueatLevel (Variables|Default|GetDashTargetTraceDefaultRadius))))

(fn GetSlamTargetTraceStart
  (return (Math|Vector|SelectVector (Pawn|GetPlayerViewPoint (Fortnite|GetFortPlayerControllerfromActor (Variables|Transient|GetAvatarPlayerPawn))) (+ (Transformation|GetActorLocation (Variables|Transient|GetAvatarPlayerPawn)) (Transformation|GetActorForwardVector (Variables|Transient|GetAvatarPlayerPawn))) (not (GameplayTags|HasAnyMatchingGameplayTags (Variables|Transient|GetAvatarPlayerPawn) (Variables|Default|GetCustomControlsTags))))))

(fn GetTraceAdjustedSpawnLocation
  (params (InSpawnLocation :type Vector))
  ;[0.02778 0.1667 0.1667] GetTraceAdjustedSpawnLocation: Adjust the spawn location of the projectile based on a trace so that the projectile does not spawn on the other side of a wall when the camera is too close to it.
  (Utilities|FlowControl|Sequence
    (:then_0
      (if (ScalableFloat|GetValueasBool (Variables|Hotfix|GetHotfix_AllowAdjustSpawnLocationBasedOnTrace))
        (bind _outfortplayerpawn (CallFunction|GetAvatarasValidFortPlayerPawn)
          (:ValidFortPlayerPawn
            (bind _output_get (LocalVariables|SetActorsToIgnore (Actor|GetAttachedActors _outfortplayerpawn true true)))
            (Utilities|Array|Add _output_get _outfortplayerpawn)
            (Utilities|Array|Add _output_get (Vehicle|GetVehicleActor _outfortplayerpawn :!class /Script/FortniteGame.FortPlayerPawn))
            (bind (_outhit _returnvalue) (Collision|LineTraceByChannel (Transformation|GetActorLocation _outfortplayerpawn) InSpawnLocation "TraceTypeQuery5" true (LocalVariables|GetActorsToIgnore) "ForDuration"))
            (if _returnvalue
              (return (.ImpactPoint (Collision|BreakHitResult _outhit))))))))
    (:then_1
      (return InSpawnLocation))))

(fn IsDirectionWithinAllowedAngle
  (params (A :type Vector))
  (return (> (Math|Vector|DotProduct A "0.000000,0.000000,-1.000000") (Variables|Default|GetMinAllowedDotProduct)) (Variables|Default|GetMinAllowedDotProduct)))

(fn OverrideFailedReason
  (params (FailedReason :type Gameplay%20Tag%20Container%20Structure%20%28by%20ref%29))
  (return 0))

(fn PlayForceFeedback
  (params (ForceFeedbackEffect :type Force%20Feedback%20Effect%20Object%20Reference))
  (bind _returnvalue (CallFunction|GetAvatarasFortPlayerPawn))
  (bind _returnvalue_2 (Fortnite|GetFortPlayerControllerfromActor _returnvalue))
  ;[0.008391 0.1308 0.1406] Play force feedback on the avatar player pawn.
  (Utilities|IsValid _returnvalue
    (:Is%20Valid
      (Utilities|IsValid _returnvalue_2
        (:Is%20Valid
          (Game|Feedback|LocalPlayForceFeedback _returnvalue_2 ForceFeedbackEffect false "None"))))))

(fn PlayHeavyCameraAnimation
  (bind _heavycameraanimation (Variables|Camera|GetHeavyCameraAnimation)
    (:then
      (bind _returnvalue (General|GetCameraAnimationModifier))
      (Utilities|IsValid _returnvalue
        (:Is%20Valid
          (bind _returnvalue_2 (CameraAnimation|PlayCameraAnimation _returnvalue _heavycameraanimation (Utilities|Struct|MakeCameraAnimationParams 1.0 1.0 "Linear" 0.2 "Linear" 1.0) :!class /Script/EngineCameras.CameraAnimationCameraModifier))
          (Variables|Transient|SetHeavyCameraAnimationHandle _returnvalue_2))))
    (:else)))

(fn PlaySlamHoverCameraAnimation
  (bind _slamhovercameraanimation (Variables|Camera|GetSlamHoverCameraAnimation)
    (:then
      (bind _returnvalue (General|GetCameraAnimationModifier))
      (Utilities|IsValid _returnvalue
        (:Is%20Valid
          (bind _returnvalue_2 (CameraAnimation|PlayCameraAnimation _returnvalue _slamhovercameraanimation (Utilities|Struct|MakeCameraAnimationParams 1.0 1.0 "Linear" 0.2 "Linear" 1.0) :!class /Script/EngineCameras.CameraAnimationCameraModifier))
          (Variables|Transient|SetSlamHoverCameraAnimationHandle _returnvalue_2))))
    (:else)))

(fn PlaySlamLandingCameraAnimation
  (bind _avatarplayerpawn (Variables|Transient|GetAvatarPlayerPawn))
  (bind _returnvalue (Fortnite|GetFortPlayerControllerfromActor _avatarplayerpawn))
  (bind _slamlandingcameraanimation (Variables|Camera|GetSlamLandingCameraAnimation)
    (:then
      (bind _returnvalue_2 (General|GetCameraAnimationModifier))
      (Utilities|IsValid _returnvalue_2
        (:Is%20Valid
          (bind _returnvalue_3 (CameraAnimation|PlayCameraAnimation _returnvalue_2 _slamlandingcameraanimation (Utilities|Struct|MakeCameraAnimationParams 1.0 1.0 "Linear" 0.2 "Linear" 1.0) :!class /Script/EngineCameras.CameraAnimationCameraModifier))
          (Variables|Transient|SetSlamLandingCameraAnimationHandle _returnvalue_3)
          (Ability|HasAuthority
            (:False
              (Utilities|IsValid _returnvalue
                (:Is%20Valid
                  (Game|Feedback|ClientStartCameraShake :self _returnvalue :Shake (Variables|Camera|GetSlamLandingCameraShake) :Scale 1.0 :PlaySpace "CameraLocal"))))))))
    (:else)))

(fn RemovePreventMoveAndSwitch
  (Ability|RemoveGameplayEffectFromOwnerWithHandle (Variables|Transient|GetPreventMovementGameplayEffectHandle))
  (Ability|RemoveGameplayEffectFromOwnerWithHandle (Variables|Transient|GetPreventEquipmentChangeGameplayEffectHandle))
  (Ability|RemoveGameplayEffectFromOwnerWithHandle (Variables|Transient|GetDivingGameplayEffectHandle)))

(fn SetEnableHighAccelerationTolerance
  (params (Enable :type Boolean))
  ;[0 0.08968 0.15 0.5] SetEnableHighAccelerationTolerance: Enable or disable excessive net-corrections during fast movements.
  (Ability|HasAuthority
    (:True
      (bind _fort_movement_comp_character (Variables|Default|GetFort%20Movement%20Comp%20Character)
        (:then
          (NetTolerance|SetUseHighAccelerationTolerance _fort_movement_comp_character Enable "VanillaSodaSlam"))
        (:else)))))

(fn SetEnableMantisCameraModifier
  (params (bEnable :type Boolean))
  (bind _mantiscameramodifierclass (Variables|Camera|GetMantisCameraModifierClass))
  ;[0.000999 0.15 0.06155 0.5] SetEnableMantisCameraModifier: Add the camera modifier for that will be used to adjust the camera during the melee attacks.
  (bind _avatarplayerpawn (Variables|Transient|GetAvatarPlayerPawn)
    (:then
      (Utilities|IsValid (Fortnite|GetFortPlayerControllerfromActor _avatarplayerpawn)
        (:Is%20Valid
          (if bEnable
            (bind _playercameramanager (Variables|/Script/Engine.PlayerController|GetPlayerCameraManager (Fortnite|GetFortPlayerControllerfromActor _avatarplayerpawn))
              (:then
                (bind _returnvalue (Game|Player|FindCameraModifierbyClass _playercameramanager _mantiscameramodifierclass))
                (Utilities|IsValid _returnvalue
                  (:Is%20Valid
                    (bind _ascamera_modifier_vanilla_soda_mantis (Utilities|Casting|CastToCameraModifier_VanillaSoda_Mantis _returnvalue))
                    (if (or (CameraModifier|IsDisabled _ascamera_modifier_vanilla_soda_mantis) (CameraModifier|IsPendingDisable _ascamera_modifier_vanilla_soda_mantis))
                      _))
                  (:Is%20Not%20Valid
                    _))
                (bind _returnvalue_2 (Game|Player|AddNewCameraModifier _playercameramanager _mantiscameramodifierclass))
                (bind _ascamera_modifier_vanilla_soda_mantis_2 (Utilities|Casting|CastToCameraModifier_VanillaSoda_Mantis _returnvalue_2))
                (Variables|Transient|SetMantisCameraModifier _ascamera_modifier_vanilla_soda_mantis_2))
              (:else))
            (else
              (bind _playercameramanager_2 (Variables|/Script/Engine.PlayerController|GetPlayerCameraManager (Fortnite|GetFortPlayerControllerfromActor _avatarplayerpawn))
                (:then
                  (bind _returnvalue_3 (Game|Player|FindCameraModifierbyClass _playercameramanager_2 _mantiscameramodifierclass))
                  (bind _ascamera_modifier_vanilla_soda_mantis_3 (Utilities|Casting|CastToCameraModifier_VanillaSoda_Mantis _returnvalue_3))
                  (Class|CameraModifierVanillaSodaMantis|DisableandRemove _ascamera_modifier_vanilla_soda_mantis_3))
                (:else)))))))
    (:else)))

(fn ToggleCanChangeWeapons
  (params (bAllow :type Boolean))
  ;[0.002269 0.15 0.107 0.5] ToggleCanChangeWeapons: Allow or disallow changing of the player's weapon/equipment.
  (if bAllow
    (Ability|RemoveGameplayEffectFromOwnerWithHandle (Variables|Transient|GetPreventEquipmentChangeGameplayEffectHandle))
    (else
      (Ability|RemoveGameplayEffectFromOwnerWithHandle (Variables|Transient|GetPreventEquipmentChangeGameplayEffectHandle))
      (bind _returnvalue (Ability|ApplyGameplayEffectToOwner (Variables|GameplayEffects|GetPreventEquipmentChangeGameplayEffectClass)))
      (Variables|Transient|SetPreventEquipmentChangeGameplayEffectHandle _returnvalue))))
