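; Gameplay ability in the VanillaSoda weapon set that, by its name, grants fall damage immunity until landing after a knockback.
; On activation it compares the avatar pawn's team affiliation with the source weapon.
; For a friendly result it ends early unless teammate knockback is allowed; otherwise it ends unless the grant-to-target setting is on.
; Activations that pass these data-row checks defer to the parent ability's activation logic.
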(event Ability|EventActivateAbility
  (if (Utilities|Enum|Equal(Enum) (Team|GetTeamAffiliation (CallFunction|GetAvatarasFortPlayerPawn) (Utilities|Casting|CastToFortWeapon (Ability|GetCurrentSourceObject))) "Friendly")
    (if (not (ScalableFloat|GetValueasBool (Variables|Default|GetRow%20Knockback%20Allow%20Teammate%20Knockback)))
      (Ability|EndAbility :!label _label))
    (elif (ScalableFloat|GetValueasBool (Variables|Default|GetRow%20Knockback%20Grant%20FDI%20to%20Target))
      _
      (else
        (goto _label))))
  (|CallParentFunction))
