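; Event-triggered gameplay ability for the VanillaSoda sword that restores health and shield to a player who scores an elimination.
; It resolves the instigator's pawn and, unless that pawn is DBNO, reads the restore amount from the ShieldStealAmount curve table row.
; The amount fills missing health first and any remainder becomes shield, applied through SetByCaller gameplay effect specs before the ability ends.
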(event Ability|EventActivateAbilityFromEvent
  (params (EventData :type Gameplay%20Event%20Data%20Structure%20%28by%20ref%29))
  (bind (_eventtag _instigator) (Utilities|Struct|BreakGameplayEventData EventData))
  (bind _amounttoheal (Variables|Default|GetAmountToHeal))
  (bind _eliminator (Variables|Default|GetEliminator))
  ;[1 1 1] Activate this ability from the event of killing a player
  (|CallParentFunction EventData)
  ;[1 1 1] Set your pawn variable once and continue to use it for the ability
  (bind _output_get (Variables|Default|SetEliminator (Game|Player|GetPlayerPawn (Fortnite|GetFortPlayerControllerfromActor _instigator))))
  (Variables|Default|SetTarget (Ability|TargetData|AbilityTargetDatafromActor _output_get))
  (if (not (Health|IsDBNO _eliminator))
    (bind _outxy (DataTable|EvaluateCurveTableRow "/VanillaSoda/DataTables/VanillaSodaGameplayGameData.VanillaSodaGameplayGameData" "VanillaSoda.Sword.ShieldStealAmount")
      (:RowFound
        (Variables|Default|SetAmountToHeal _outxy)
        (Variables|Default|SetHealthToHeal (Math|Float|SelectFloat _amounttoheal (- (FortPawn|GetMaxHealth _eliminator) (FortPawn|GetHealth _eliminator)) (>= (- (FortPawn|GetMaxHealth _eliminator) (FortPawn|GetHealth _eliminator)) _amounttoheal)))
        (Variables|Default|SetHealthToShield (- _amounttoheal (Variables|Default|GetHealthToHeal)))
        (if (< (FortPawn|GetHealth _eliminator) (FortPawn|GetMaxHealth _eliminator))
          (bind _returnvalue (Ability|GameplayEffect|AssignTagSetbyCallerMagnitude (Ability|MakeOutgoingGameplayEffectSpec (Variables|Default|GetHealGE)) "(TagName=\"SetByCaller.Healing\")" (Variables|Default|GetHealthToHeal)))
          (Ability|ApplyGameplayEffectSpecToTarget _returnvalue (Variables|Default|GetTarget))
          (else
            _))
        (bind _returnvalue_2 (Ability|GameplayEffect|AssignTagSetbyCallerMagnitude (Ability|MakeOutgoingGameplayEffectSpec (Variables|Default|GetShieldGE)) "(TagName=\"SetByCaller.Shield\")" (Variables|Default|GetHealthToShield)))
        (Ability|ApplyGameplayEffectSpecToTarget _returnvalue_2 (Variables|Default|GetTarget)))))
  (Ability|EndAbility))

(fn ShouldAbilityRespondToEvent
  (params (ActorInfo :type Gameplay%20Ability%20Actor%20Info%20Structure) (Payload :type Gameplay%20Event%20Data%20Structure))
  (return true))
