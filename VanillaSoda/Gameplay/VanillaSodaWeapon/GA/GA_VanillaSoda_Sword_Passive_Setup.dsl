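; Passive gameplay ability that sets up the VanillaSoda sword wielder with max health and shield buffs, regeneration and altered HUD visibility.
; On activation it applies buffs through the pawn component, tags the pawn next frame to block repeat initial heals, and starts a looping regen timer.
; Taking damage applies a cooldown effect; otherwise, unless DBNO, each tick regenerates health, then shield once health is full.
; Ending restores HUD and removes buffs.
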(event Ability|EventActivateAbility
  (bind _playerpawn (Variables|Default|GetPlayerPawn))
  (bind _returnvalue (Actor|GetComponentbyClass _playerpawn "/VanillaSoda/Gameplay/VanillaSodaWeapon/Components/BP_FortPawnComponent_VanillaSoda.BP_FortPawnComponent_VanillaSoda_C"))
  (|CallParentFunction)
  ;[1 1 1] Initialize.
  (bind _asplayer_pawn_athena (Utilities|Casting|CastToPlayerPawn_Athena (Ability|GetAvatarActorfromActorInfo))
    (:then
      (Variables|Default|SetPlayerPawn _asplayer_pawn_athena)
      (CallFunction|DisplayHUD false)
      (EventDispatchers|BindEventtoOnDamaged (Variables|Default|GetPlayerPawn) (delegate Custom|EventDamaged))
      ;[0.04549 0.2349 1] Handle Regen
      ;;[1 1 1] Assign TickRate through rows in AthenaGameData
      (Variables|Default|SetTickRate (ScalableFloat|GetValueatLevel (Variables|Default|GetTickRateRow)))
      ;;[1 1 1] Setup timer at the speed of tick rate
      (Utilities|Time|SetTimerbyEvent (EventDispatchers|CreateEvent :!event Custom|ApplyHealthOrShield) (Variables|Default|GetTickRate) true)
      (Utilities|IsValid _returnvalue
        (:Is%20Valid
          ; TODO: Hookup scalable floats to HF data table.
          (Class|FortPawnComponentVanillaSoda|ApplyBuffs _returnvalue (ScalableFloat|GetValueatLevel (Variables|Default|GetMaxHealthBuff)) (ScalableFloat|GetValueatLevel (Variables|Default|GetMaxShieldBuff)) false false)
          ; Skips a bit to ensure the heal has already happened, then add a tag to ensure it doesn't happen again
          (Utilities|FlowControl|DelayUntilNextFrame)
          (Ability|GameplayEffect|AddLooseGameplayTags (Variables|Default|GetPlayerPawn) (GameplayTags|MakeGameplayTagContainerfromTag (Variables|Default|GetInitialHealBlockerTag)) true))))
    (:CastFailed
      (Development|PrintString "Avatar Actor is NOT PlayerPawn_Athena"))))

(event Custom|EventDamaged
  (params (DamagedActor :type Actor%20Object%20Reference) (Damage :type Float%20%28single-precision%29) (InstigatedBy :type Controller%20Object%20Reference) (DamageCauser :type Actor%20Object%20Reference) (HitLocation :type Vector) (FHitComponent :type Primitive%20Component%20Object%20Reference) (BoneName :type Name) (Momentum :type Vector))
  ;[1 1 1] Apply Regen Cooldown when damaged.
  (Ability|ApplyGameplayEffectToOwner "/Game/Athena/Items/Gameplay/Sword/GE_DamagedTag.GE_DamagedTag_C" 0))

(event Ability|EventFailedToActivatePassiveAbility
  (params (ActorInfo :type Gameplay%20Ability%20Actor%20Info%20Structure))
  (|CallParentFunction))

(event Ability|EventOnEndAbility
  (params (bWasCancelled :type Boolean))
  (bind _playerpawn (Variables|Default|GetPlayerPawn))
  (bind _returnvalue (Actor|GetComponentbyClass _playerpawn "/VanillaSoda/Gameplay/VanillaSodaWeapon/Components/BP_FortPawnComponent_VanillaSoda.BP_FortPawnComponent_VanillaSoda_C"))
  (CallFunction|DisplayHUD true)
  (GameplayTasks|EndTask (Variables|Default|GetTagChangeTask))
  (Utilities|IsValid _returnvalue
    (:Is%20Valid
      (Class|FortPawnComponentVanillaSoda|RemoveBuffs _returnvalue false false))))

(event Custom|ApplyHealthOrShield
  (bind _playerpawn (Variables|Default|GetPlayerPawn))
  ;[1 1 1] Grant health or shield if not on cooldown.
  (if (GameplayTags|HasMatchingGameplayTag _playerpawn "(TagName=\"Athena.Blade.Damaged\")")
    _
    (elif (Health|IsDBNO _playerpawn)
      _
      (elif (< (FortPawn|GetHealth _playerpawn) (FortPawn|GetMaxHealth _playerpawn))
        ;;[0.2686 1 0.3384] Health
        (Ability|ApplyGameplayEffectToOwner "/Game/Athena/Items/Gameplay/Sword/GE_Athena_Blade_HealthRegen.GE_Athena_Blade_HealthRegen_C" 0)
        (else
          ;;[0.286 0.7239 1] Shield
          (Ability|ApplyGameplayEffectToOwner "/Game/Athena/Items/Gameplay/Sword/GE_Athena_Blade_Shield_Regen.GE_Athena_Blade_Shield_Regen_C" 0))))))

(fn DisplayHUD
  (params (Show :type Boolean))
  (bind _playerpawn (Variables|Default|GetPlayerPawn)
    (:then
      (Utilities|IsValid (Fortnite|GetFortPlayerControllerfromActor _playerpawn)
        (:Is%20Valid
          (HUD|SetHUDElementVisibilitywithContext (Fortnite|GetFortPlayerControllerfromActor _playerpawn) (GameplayTags|MakeGameplayTagContainerfromTag "(TagName=\"HUD.Athena.HidePrimaryForAbility\")") (not Show) "None" "")
          (return))))
    (:else)))
